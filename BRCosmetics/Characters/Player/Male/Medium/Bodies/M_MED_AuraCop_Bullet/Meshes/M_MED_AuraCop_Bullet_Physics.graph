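# Physics asset for the male medium AuraCop Bullet player body, covering a dangling necklace and the upper spine it hangs from.
# Capsules on spine_04 and spine_05 form the torso volume, and smaller capsules cover the necklace base and three dynamic links.
# The base joins spine_05 with free swing and twist, while each link below chains to the previous one with narrow swing limits and locked twist.
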PhysicsAsset /BRCosmetics/Characters/Player/Male/Medium/Bodies/M_MED_AuraCop_Bullet/Meshes/M_MED_AuraCop_Bullet_Physics
bodies {
 base_necklace [{shapeType=Capsule, length=0.505, radius=0.505, shapeName=base_necklace_capsule, rotation=<rot 0 -90 0>}]
 dyn_necklace_01 [{
   shapeType=Capsule
   length=10.3535
   radius=0.55
   shapeName=dyn_necklace_01_capsule
   center=<1.7028 -0.0504 0.0677>
   rotation=<rot 0.3347 -90 0.8787>
  }]
 dyn_necklace_02 [{shapeType=Capsule, length=6.9018, radius=0.55, shapeName=dyn_necklace_02_capsule, center=<0.9576 -0.0172 -0.2272>}]
 dyn_necklace_03 [{shapeType=Capsule, length=4.871, radius=1.5, shapeName=dyn_necklace_03_capsule, center=<3.848 -0.0765 -0.2287>, rotation=<rot 90 0 0>}]
 spine_04 [{shapeType=Capsule, length=14.9758, radius=6, center=<6.9287 4.8056 -0.162>, rotation=<rot 0.057 -90 0.938>}]
 spine_05 [{shapeType=Capsule, length=14.4571, radius=7.5, center=<1.9165 1.2799 -0.0073>}]
}
constraints [
 {bone1Name=spine_05, bone2Name=spine_04, swing1LimitDegrees=45, swing2LimitDegrees=45, twistLimitDegrees=45}
 {
  bone1Name=base_necklace
  bone2Name=spine_05
  swing1LimitDegrees=45
  swing1Motion=Free
  swing2LimitDegrees=45
  swing2Motion=Free
  twistLimitDegrees=45
  twistMotion=Free
 }
 {bone1Name=dyn_necklace_01, bone2Name=base_necklace, swing1LimitDegrees=20, swing2LimitDegrees=10, twistLimitDegrees=45, twistMotion=Locked}
 {bone1Name=dyn_necklace_02, bone2Name=dyn_necklace_01, swing1LimitDegrees=20, swing2LimitDegrees=10, twistLimitDegrees=45, twistMotion=Locked}
 {bone1Name=dyn_necklace_03, bone2Name=dyn_necklace_02, swing1LimitDegrees=20, swing2LimitDegrees=10, twistLimitDegrees=45, twistMotion=Locked}
]
