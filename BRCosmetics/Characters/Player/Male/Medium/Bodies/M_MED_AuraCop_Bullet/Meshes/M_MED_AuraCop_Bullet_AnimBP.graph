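# Post-processes the M_MED_AuraCop_Bullet player body cosmetic, adding secondary motion to the pose copied from its attached parent.
# A Control Rig feeds weighted multi-bone constraints on both gun bases and, during front-end emotes, the necklace base.
# Additive offsets nudge the necklace during skydiving before parachute opening and by head-to-spine angle.
# RigidBody physics then simulates in spine_05 space, with gravity overridden each update by a deprecated state function.
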AnimBlueprint /BRCosmetics/Characters/Player/Male/Medium/Bodies/M_MED_AuraCop_Bullet/Meshes/M_MED_AuraCop_Bullet_AnimBP
target_skeleton=&/M_MED_AuraCop_Bullet_Skeleton
graphs [
 AnimGraph {nodes [
   n1 = Control Rig(Source=Copy Pose From Mesh() {bUseAttachedParent=true}) {
    alphaBoolBlend {blendOption=Linear}
    controlRigAssetReference {blueprintRigClass=&/M_MED_AuraCop_Bullet_CtrlRig.M_MED_AuraCop_Bullet_CtrlRig_C}
   }
   n2 = "Constraint - base_necklace (PPPPP)"!(
    ComponentPose=Local To Component(LocalPose=@n1)
    Alpha=Get FE_Emoting() -> (FE_Emoting:double)
    ConstraintWeights_0=0.4382
    ConstraintWeights_1:float=0.328
    ConstraintWeights_2:float=0.1
    ConstraintWeights_3:float=0.0874
    ConstraintWeights_4:float=0.0464
   ) -> (Pose) {
    alphaBoolBlend {blendOption=Linear}
    boneToModify {boneName=base_necklace}
    constraintSetup [
     {targetBone {boneName=neck_01}}
     {targetBone {boneName=spine_05}}
     {targetBone {boneName=neck_02}}
     {targetBone {boneName=spine_04}}
     {targetBone {boneName=head}}
    ]
    constraintWeights [1,1,1,1,1]
   }
   n3 = "Constraint - base_gun_l (PPPPPPPP)"!(
    ComponentPose=@n2
    Alpha=1
    ConstraintWeights_0=0.3013
    ConstraintWeights_1:float=0.2202
    ConstraintWeights_2:float=0.1538
    ConstraintWeights_3:float=0.1107
    ConstraintWeights_4:float=0.0865
    ConstraintWeights_5:float=0.0571
    ConstraintWeights_6:float=0.0451
    ConstraintWeights_7:float=0.0252
   ) -> (Pose) {
    alphaBoolBlend {blendOption=Linear}
    boneToModify {boneName=base_gun_l}
    constraintSetup [
     {targetBone {boneName=spine_03}}
     {targetBone {boneName=spine_02}}
     {targetBone {boneName=spine_04}}
     {targetBone {boneName=clavicle_l}}
     {targetBone {boneName=spine_01}}
     {targetBone {boneName=deform_pec_l}}
     {targetBone {boneName=upperarm_twist_01_l}}
     {targetBone {boneName=pelvis}}
    ]
    constraintWeights [1,1,1,1,1,1,1,1]
   }
   n4 = "Constraint - base_gun_r (PPPPPPPP)"!(
    ComponentPose=@n3
    Alpha=1
    ConstraintWeights_0=0.303
    ConstraintWeights_1:float=0.2243
    ConstraintWeights_2:float=0.1523
    ConstraintWeights_3:float=0.1069
    ConstraintWeights_4:float=0.0903
    ConstraintWeights_5:float=0.0542
    ConstraintWeights_6:float=0.0408
    ConstraintWeights_7:float=0.0282
   ) -> (Pose) {
    alphaBoolBlend {blendOption=Linear}
    boneToModify {boneName=base_gun_r}
    constraintSetup [
     {targetBone {boneName=spine_03}}
     {targetBone {boneName=spine_02}}
     {targetBone {boneName=spine_04}}
     {targetBone {boneName=clavicle_r}}
     {targetBone {boneName=spine_01}}
     {targetBone {boneName=deform_pec_r}}
     {targetBone {boneName=upperarm_twist_01_r}}
     {targetBone {boneName=pelvis}}
    ]
    constraintWeights [1,1,1,1,1,1,1,1]
   }
   n5 = "Transform (Modify) Bone - Bone: base_necklace"(ComponentPose=@n4, Alpha=Get Skydive alpha() -> (Skydive alpha:double), Translation="0,-1.5,0", Rotation="0,0,0", Scale="1,1,1") {
    rotationMode=BMM_Additive
    rotationSpace=BCS_BoneSpace
    translationMode=BMM_Additive
    translationSpace=BCS_BoneSpace
    alphaBoolBlend {blendOption=Linear}
    boneToModify {boneName=base_necklace}
    translation {y=-1.5}
   }
   n6 = "Transform (Modify) Bone - Bone: base_necklace"(ComponentPose=@n5, Alpha=Get Head Up() -> (Head Up:float), Translation="0,-1.5,0", Rotation="0,2,0", Scale="1,1,1") {
    rotationMode=BMM_Additive
    rotationSpace=BCS_BoneSpace
    translationMode=BMM_Additive
    translationSpace=BCS_BoneSpace
    alphaBoolBlend {blendOption=Linear}
    boneToModify {boneName=base_necklace}
    rotation {yaw=2}
    translation {y=-1.5}
   }
   n7 = RigidBody!(ComponentPose=@n6, Alpha, OverrideWorldGravity:Vector=Get GravityOverride(), ExternalForce) -> (Pose) {
    bClampLinearTranslationLimitToRefPose=true
    bOverrideWorldGravity=true
    overridePhysicsAsset=&/M_MED_AuraCop_Bullet_Physics
    simulationSpace=BaseBoneSpace
    alphaBoolBlend {blendOption=Linear}
    baseBoneRef {boneName=spine_05}
    componentAppliedLinearAccClamp=<3000 3000 3000>
    componentLinearAccScale=<0.4 0.3 0.3>
    componentLinearVelScale=<0.4 0.3 0.3>
    overrideWorldGravity {z=-980}
    simSpaceSettings {
     maxAngularAcceleration=1000
     maxAngularVelocity=3
     maxLinearAcceleration=700
     maxLinearVelocity=500
     velocityScaleZ=0
     worldAlpha=0.5
     externalLinearDragV=<0.2 0.2>
    }
   }
   Output Pose(Result=Component To Local(ComponentPose=@n7)) {layerGroup=None}
  ]}
 EventGraph {nodes [
   Event BlueprintUpdateAnimation()
   > n1 = Gravity Override State Function DEPRECATED(CharacterPartAnimInstance=Self-Reference(), GameTimeInSeconds=GetGameTimeInSeconds(), GravityOverrideParams=Get Gravity Override Params())
   > Set GravityOverride(GravityOverride=@n1)
   n2 = NOT Boolean(A=Get bIsParachuteOpened())
   n3 = Get bIsSkyDiving()
   > Set Skydive alpha(Skydive alpha:double="To Float (Boolean)"(InBool=AND Boolean(A=@n3, B=@n2))) -> (Output_Get:double)
   n4 = Get bIsPlayingEmote()
   n5 = Get bIsInFrontEnd()
   > Set FE_Emoting(FE_Emoting:double="To Float (Boolean)"(InBool=AND Boolean(A=@n5, B=@n4))) -> (Output_Get:double)
   > Set Head Up(Head Up:float=GetAlphaFromReferenceTransformDelta(SkeletalMeshComponent=Get SkelMesh(), BoneName=spine_04, BaseName=head, InputRange=(X=15,Y=60), TwistAxis="0,0,1")) -> (Output_Get:float)
   Event BlueprintInitializeAnimation()
   > Set SkelMesh(SkelMesh=GetOwnerSkeletalMeshComponent())
  ]}
]
variables [
 FE_Emoting {type=number}
 Gravity Override Params {type=GravityOverrideParamsStruct}
 GravityOverride {type=Vector}
 Head Up {type=number}
 SkelMesh {type=SkeletalMeshComponent}
 Skydive alpha {type=number}
]
