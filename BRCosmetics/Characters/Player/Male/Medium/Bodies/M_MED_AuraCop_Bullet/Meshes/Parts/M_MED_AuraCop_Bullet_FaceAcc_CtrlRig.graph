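# Control Rig for the male AuraCop Bullet face-accessory part that applies corrective offsets to hair and clavicle helper bones.
# A sequence runs debug-only control-to-bone copies, then spherical pose readers on the head, relative to spine bones, weight hair-bone offsets, some gated by Is Targeting Crouch.
# Remapped clavicle rotation weights the left mover bone's offset, while the right mover uses constant full weight, leaving its computed delta unused.
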ControlRigBlueprint /BRCosmetics/Characters/Player/Male/Medium/Bodies/M_MED_AuraCop_Bullet/Meshes/Parts/M_MED_AuraCop_Bullet_FaceAcc_CtrlRig
graphs [
 New Function_ContainedGraph {nodes [
   n1 = RigVMFunctionEntryNode()
   RigVMFunction_ControlFlowBranch = &RigVMFunction_ControlFlowBranch(Condition=@n1.Debug, ExecuteContext=@n1.ExecuteContext)
   RigVMFunctionReturnNode(ExecuteContext=@RigVMFunction_ControlFlowBranch.Completed)
   RigUnit_GetTransform = &RigUnit_GetTransform(Item=(Type=Control,Name="head_ctrl"), bInitial=False)
   &RigUnit_SetTransform(
    Item=(Type=Bone,Name="head")
    Weight=1
    bInitial=False
    Value=@RigUnit_GetTransform.Transform
    ExecutePin=@RigVMFunction_ControlFlowBranch.True
   )
  ]}
 New Function_ContainedGraph {nodes [
   n1 = RigVMFunctionEntryNode()
   RigVMFunction_ControlFlowBranch = &RigVMFunction_ControlFlowBranch(Condition=@n1.Debug, ExecuteContext=@n1.ExecuteContext)
   RigVMFunctionReturnNode(ExecuteContext=@RigVMFunction_ControlFlowBranch.Completed)
   RigUnit_GetTransform = &RigUnit_GetTransform(Item=(Type=Control,Name="spine_05_ctrl"), bInitial=False)
   RigUnit_SetTransform = &RigUnit_SetTransform(
    Item=(Type=Bone,Name="spine_05")
    Weight=1
    bInitial=False
    Value=@RigUnit_GetTransform.Transform
    ExecutePin=@RigVMFunction_ControlFlowBranch.True
   )
   > &RigUnit_SetTransform(
    Item=(Type=Bone,Name="neck_01")
    Weight=1
    bInitial=False
    Value=&RigUnit_GetTransform(Item=(Type=Control,Name="neck_01_ctrl"), bInitial=False).Transform
   )
   > &RigUnit_SetTransform(
    Item=(Type=Bone,Name="neck_02")
    Weight=1
    bInitial=False
    Value=&RigUnit_GetTransform(Item=(Type=Control,Name="neck_02_ctrl"), bInitial=False).Transform
   )
   > &RigUnit_SetTransform(
    Item=(Type=Bone,Name="head")
    Weight=1
    bInitial=False
    Value=&RigUnit_GetTransform(Item=(Type=Control,Name="head_ctrl"), bInitial=False).Transform
   )
   > &RigUnit_SetTransform(
    Item=(Type=Bone,Name="clavicle_l")
    Weight=1
    bInitial=False
    Value=&RigUnit_GetTransform(Item=(Type=Control,Name="clavicle_l_ctrl"), bInitial=False).Transform
   )
   > &RigUnit_SetTransform(
    Item=(Type=Bone,Name="clavicle_r")
    Weight=1
    bInitial=False
    Value=&RigUnit_GetTransform(Item=(Type=Control,Name="clavicle_r_ctrl"), bInitial=False).Transform
   )
  ]}
 New Function_ContainedGraph {nodes [
   n1 = RigVMFunctionEntryNode()
   RigVMFunction_MathQuaternionToEuler = &RigVMFunction_MathQuaternionToEuler(
    RotationOrder=ZYX
    Value=&RigVMFunction_MathTransformMakeRelative(
     Global=&RigUnit_GetTransform(Item=(Type=Bone,Name="foot_l"), Space=LocalSpace, Item.Name=@n1.Bone Name).Transform
     Parent=&RigUnit_GetTransform(Item=(Type=Bone,Name="foot_l"), Space=LocalSpace, bInitial=true, Item.Name=@n1.Bone Name).Transform
    ).Local.Rotation
   )
   Remap = &RigVMFunction_MathDoubleRemap(
    TargetMaximum=1
    TargetMinimum=0
    SourceMaximum=@n1.Max Source
    SourceMinimum=@n1.Min Source
    Value=@RigVMFunction_MathQuaternionToEuler.Result.Y
   )
   > RigVMFunctionReturnNode(Delta=@Remap.Result)
   RigVMDispatchNode(Enabled=True, ExecuteContext=(), ScreenColor=(R=1,G=1,B=1,A=1), ScreenDuration=0.05, Value=@RigVMFunction_MathQuaternionToEuler.Result.Y) {template="DISPATCH_RigVMDispatch_Print(in Prefix,in Value,in Enabled,in ScreenDuration,in ScreenColor)"}
  ]}
 RigVMFunctionLibrary {nodes [
   Controls = RigVMCollapseNode(Debug=False, ExecuteContext=())
   Controls_1 = RigVMCollapseNode(Debug=False, ExecuteContext=())
   Delta Rotation = RigVMCollapseNode(ExecuteContext=())
  ]}
 RigVMFunction_Sequence_1_ContainedGraph {nodes [
   n1 = RigVMFunctionEntryNode()
   RigVMFunction_Sequence = &RigVMFunction_Sequence(ExecuteContext=@n1.ExecuteContext)
   n2 = &RigVMFunction_Sequence(ExecuteContext=@RigVMFunction_Sequence.B)
   n3 = &RigVMFunction_Sequence(ExecuteContext=@n2.B)
   n4 = &RigVMFunction_Sequence(ExecuteContext=@n3.B)
   RigVMFunctionReturnNode(A=@RigVMFunction_Sequence.A, B=@n2.A, C=@n3.A, D=@n4.A, E=@n4.B)
  ]}
 RigVMModel {nodes [
   RigUnit_BeginExecution = &RigUnit_BeginExecution()
   RigVMFunction_Sequence_2 = RigVMAggregateNode(ExecuteContext=@RigUnit_BeginExecution.ExecutePin)
   Controls_1 = RigVMFunctionReferenceNode(Debug=false, ExecuteContext=@RigVMFunction_Sequence_2.A)
   n1 = &RigUnit_SphericalPoseReader(
    ActiveRegionScaleFactors=(PositiveWidth=0.6,NegativeWidth=0.6,PositiveHeight=8,NegativeHeight=0.07)
    ActiveRegionSize=0.1
    Debug=(bDrawDebug=false,bDraw2D=False,bDrawLocalAxes=False,DebugScale=25,DebugSegments=20,DebugThickness=0.25)
    DriverAxis=(X=1,Y=0,Z=0)
    DriverItem=(Type=Bone,Name="head")
    FalloffRegionScaleFactors=(PositiveWidth=0.8,NegativeWidth=0.8,PositiveHeight=0,NegativeHeight=0.8)
    FalloffSize=0.25
    FlipHeightScaling=False
    FlipWidthScaling=False
    OptionalParentItem=(Type=Bone,Name="spine_05")
    RotationOffset=(X=0,Y=-90,Z=0)
    ExecutePin=@RigVMFunction_Sequence_2.B
   )
   n2 = &RigVMFunction_ControlFlowBranch(Condition=false, ExecuteContext=@n1.ExecutePin)
   n3 = &RigVMFunction_MathDoubleSub(A=1, B=@n1.OutputParam)
   &RigUnit_ModifyTransforms(
    ItemToModify="((Item=(Type=Bone,Name=\"base_hair_02_l\"),Transform=(Rotation=(X=0,Y=0,Z=-0.342,W=0.9397),Translation=(X=0,Y=0,Z=0),Scale3D=(X=1,Y=1,Z=1))),(Item=(Type=Bone,Name=\"dyn_hair_03_l\"),Transform=(Rotation=(X=0,Y=0,Z=0,W=1),Translation=(X=0,Y=0,Z=0),Scale3D=(X=1,Y ...(truncated)"
    WeightMinimum=0
    ExecutePin=@n2.False
    Weight=@n3.Result
   )
   > &RigUnit_ModifyTransforms(
    ItemToModify=((Item=(Type=Bone,Name="base_hair_02_m"),Transform=(Rotation=(X=0,Y=0,Z=-0.3827,W=0.9239),Translation=(X=0,Y=0,Z=0),Scale3D=(X=1,Y=1,Z=1))))
    WeightMinimum=0
    Weight=@n3.Result
   )
   > &RigUnit_ModifyTransforms(
    ItemToModify="((Item=(Type=Bone,Name=\"base_hair_02_r\"),Transform=(Rotation=(X=-0.1178,Y=0.1797,Z=-0.3126,W=0.9253),Translation=(X=0,Y=0,Z=0),Scale3D=(X=1,Y=1,Z=1))),(Item=(Type=Bone,Name=\"dyn_hair_03_r\"),Transform=(Rotation=(X=0,Y=0,Z=0,W=1),Translation=(X=0,Y=0,Z=0),Sc ...(truncated)"
    WeightMinimum=0
    Weight=@n3.Result
   )
   Delta Rotation = RigVMFunctionReferenceNode(Bone Name=clavicle_l, Max Source=-47, ExecuteContext=@RigVMFunction_Sequence_2.C)
   > &RigUnit_ModifyTransforms(
    ItemToModify=((Item=(Type=Bone,Name="dyn_clavicle_mover_l"),Transform=(Rotation=(X=0,Y=0,Z=0,W=1),Translation=(X=0,Y=0,Z=-3),Scale3D=(X=1,Y=1,Z=1))))
    WeightMinimum=0
    Weight=@"Delta Rotation".Delta
   )
   > Delta Rotation_1 = RigVMFunctionReferenceNode(Bone Name=clavicle_r, Max Source=-47)
   > &RigUnit_ModifyTransforms(
    ItemToModify=((Item=(Type=Bone,Name="dyn_clavicle_mover_r"),Transform=(Rotation=(X=0,Y=0,Z=0,W=1),Translation=(X=0,Y=0,Z=3),Scale3D=(X=1,Y=1,Z=1))))
    Weight=1
    WeightMinimum=0
   )
   n4 = &RigUnit_SphericalPoseReader(
    ActiveRegionScaleFactors=(PositiveWidth=1.1,NegativeWidth=1,PositiveHeight=0.2,NegativeHeight=1)
    ActiveRegionSize=0.1
    Debug=(bDrawDebug=true,bDraw2D=False,bDrawLocalAxes=False,DebugScale=25,DebugSegments=20,DebugThickness=0.25)
    DriverAxis=(X=1,Y=0,Z=0)
    DriverItem=(Type=Bone,Name="head")
    FalloffRegionScaleFactors=(PositiveWidth=1,NegativeWidth=1,PositiveHeight=1,NegativeHeight=1)
    FalloffSize=0.25
    FlipHeightScaling=False
    FlipWidthScaling=False
    OptionalParentItem=(Type=Bone,Name="spine_02")
    RotationOffset=(X=-50,Y=0,Z=0)
    ExecutePin=@RigVMFunction_Sequence_2.E
   )
   RigVMFunction_MathBoolToFloat = &RigVMFunction_MathBoolToFloat(Value=RigVMVariableNode() {variable=Is Targeting Crouch}.Value)
   Multiply = &RigVMFunction_MathDoubleMul(B=@RigVMFunction_MathBoolToFloat.Result, A=@n4.OutputParam)
   > &RigUnit_ModifyTransforms(
    ItemToModify="((Item=(Type=Bone,Name=\"base_hair_02_l\"),Transform=(Rotation=(X=0,Y=0,Z=0,W=1),Translation=(X=0,Y=0,Z=0),Scale3D=(X=1,Y=1,Z=1))),(Item=(Type=Bone,Name=\"base_hair_02_m\"),Transform=(Rotation=(X=0,Y=0,Z=-0.2588,W=0.9659),Translation=(X=0,Y=-2,Z=0),Scale3D=(X= ...(truncated)"
    WeightMinimum=0
    Weight=@Multiply.Result
   )
   n5 = &RigUnit_SphericalPoseReader(
    ActiveRegionScaleFactors=(PositiveWidth=3.2,NegativeWidth=1,PositiveHeight=1,NegativeHeight=1)
    ActiveRegionSize=0.1
    Debug=(bDrawDebug=false,bDraw2D=False,bDrawLocalAxes=False,DebugScale=25,DebugSegments=20,DebugThickness=0.25)
    DriverAxis=(X=1,Y=0,Z=0)
    DriverItem=(Type=Bone,Name="head")
    FalloffRegionScaleFactors=(PositiveWidth=1.4,NegativeWidth=1,PositiveHeight=1,NegativeHeight=1)
    FalloffSize=0.4
    FlipHeightScaling=False
    FlipWidthScaling=False
    OptionalParentItem=(Type=Bone,Name="spine_05")
    RotationOffset=(X=-20,Y=0,Z=0)
    ExecutePin=@RigVMFunction_Sequence_2.D
   )
   > n6 = &RigUnit_ModifyTransforms(
    ItemToModify="((Item=(Type=Bone,Name=\"base_hair_r\"),Transform=(Rotation=(X=0,Y=0,Z=0,W=1),Translation=(X=0,Y=0,Z=0),Scale3D=(X=1,Y=1,Z=1))),(Item=(Type=Bone,Name=\"base_hair_02_r\"),Transform=(Rotation=(X=-0.3304,Y=0.0885,Z=-0.2432,W=0.9077),Translation=(X=0,Y=0,Z=0),Scal ...(truncated)"
    WeightMinimum=0
    Weight=@n5.OutputParam
   )
   n7 = &RigVMFunction_ControlFlowBranch(ExecuteContext=@n6.ExecutePin, Condition=RigVMVariableNode() {variable=Is Targeting Crouch}.Value)
   &RigUnit_ModifyTransforms(
    ItemToModify="((Item=(Type=Bone,Name=\"base_hair_r\"),Transform=(Rotation=(X=0,Y=0,Z=0.3827,W=0.9239),Translation=(X=0,Y=-1,Z=0),Scale3D=(X=1,Y=1,Z=1))),(Item=(Type=Bone,Name=\"base_hair_02_r\"),Transform=(Rotation=(X=0,Y=0,Z=0,W=1),Translation=(X=1,Y=0,Z=-1),Scale3D=(X=1,Y ...(truncated)"
    WeightMinimum=0
    Weight=&RigVMFunction_MathDoubleMul(B=1, A=@n5.OutputParam).Result
    ExecutePin=@n7.True
   )
   &RigUnit_ModifyTransforms(
    ExecutePin=()
    ItemToModify="((Item=(Type=Bone,Name=\"base_hair_02_l\"),Transform=(Rotation=(X=0,Y=0,Z=-0.4226,W=0.9063),Translation=(X=0,Y=0,Z=0),Scale3D=(X=1,Y=1,Z=1))),(Item=(Type=Bone,Name=\"dyn_hair_03_l\"),Transform=(Rotation=(X=0,Y=0,Z=-0.1322,W=0.9912),Translation=(X=0,Y=0,Z=0),Sc ...(truncated)"
    WeightMinimum=0
    Weight=@n3.Result
   )
   &RigUnit_ModifyTransforms(
    ExecutePin=()
    ItemToModify="((Item=(Type=Bone,Name=\"base_hair_02_l\"),Transform=(Rotation=(X=0,Y=0,Z=-0.1736,W=0.9848),Translation=(X=0,Y=0,Z=0),Scale3D=(X=1,Y=1,Z=1))),(Item=(Type=Bone,Name=\"base_hair_02_m\"),Transform=(Rotation=(X=0,Y=0,Z=-0.1736,W=0.9848),Translation=(X=0,Y=-1,Z=0), ...(truncated)"
    WeightMinimum=0
    Weight=@n4.OutputParam
   )
  ]}
]
hierarchy [
 root {loc=<0 0 0>, rot=<0 0 0 1>}
 pelvis {parent=root, loc=<0 -0.7071 93.1191>, rot=<0.0127 0.707 -0.0127 -0.707>}
 spine_01 {parent=pelvis, loc=<6.7494 -0.6388 0>, rot=<0 0 0.034 -0.9994>}
 spine_02 {parent=spine_01, loc=<7.252 -0.035 0>, rot=<0 0 0.0238 -0.9997>}
 spine_03 {parent=spine_02, loc=<6.0106 0 0>, rot=<0 0 0.0149 -0.9999>}
 spine_04 {parent=spine_03, loc=<8.1023 0 0>, rot=<0 0 -0.0132 -0.9999>}
 spine_05 {parent=spine_04, loc=<10.0142 0 0>, rot=<0 0 -0.0015 -1>}
 neck_01 {parent=spine_05, loc=<11.3138 0.1529 0>, rot=<0 0 -0.2003 -0.9797>}
 neck_02 {parent=neck_01, loc=<4.5627 0 0>, rot=<0 0 0.0504 -0.9987>}
 head {parent=neck_02, loc=<5.3914 0 0>, rot=<0 0 0.115 -0.9934>}
 base_hair_m {parent=head, loc=<8.5033 -7.546 0.3187>, rot=<0.0264 -0.9997 0 0>}
 dyn_hair_01_m {parent=base_hair_m, loc=<0 0 0>, rot=<0 0 0 1>}
 base_hair_02_m {parent=dyn_hair_01_m, loc=<4.6898 0 0>, rot=<0 0 0 1>}
 dyn_hair_02_m {parent=base_hair_02_m, loc=<0 0 0>, rot=<0 0 0 1>}
 dyn_hair_03_m {parent=dyn_hair_02_m, loc=<4.6898 0 0>, rot=<0 0 0 1>}
 dyn_hair_04_m {parent=dyn_hair_03_m, loc=<4.6898 0 0>, rot=<0 0 0 1>}
 base_hair_r {parent=head, loc=<8.2695 -3.1203 5.1601>, rot=<0.0264 -0.9997 0 0>}
 dyn_hair_01_r {parent=base_hair_r, loc=<0 0 0>, rot=<0 0 0 1>}
 base_hair_02_r {parent=dyn_hair_01_r, loc=<4.6898 0 0>, rot=<0 0 0 1>}
 dyn_hair_02_r {parent=base_hair_02_r, loc=<0 0 0>, rot=<0 0 0 1>}
 dyn_hair_03_r {parent=dyn_hair_02_r, loc=<4.6898 0 0>, rot=<0 0 0 1>}
 dyn_hair_04_r {parent=dyn_hair_03_r, loc=<4.6898 0 0>, rot=<0 0 0 1>}
 hairSmall_master_r {parent=dyn_hair_04_r, loc=<-3.732 0.4741 -0.9322>, rot=<-0.3426 -0.5065 -0.0716 -0.788>}
 hairSmall_1_01_r {parent=hairSmall_master_r, loc=<2.0384 -0.1208 -1.4887>, rot=<0.1078 -0.1995 0.1729 -0.9585>}
 hairSmall_3_01_r {parent=hairSmall_master_r, loc=<2.0372 1.549 1.1925>, rot=<-0.0691 0.0214 -0.2748 -0.9588>}
 hairSmall_2_01_r {parent=hairSmall_master_r, loc=<2.0199 -0.3017 1.9764>, rot=<-0.0069 0.0372 0.068 -0.997>}
 base_hair_l {parent=head, loc=<8.2695 -3.1203 -5.16>, rot=<0.0264 -0.9997 0 0>}
 dyn_hair_01_l {parent=base_hair_l, loc=<0 0 0>, rot=<0 0 0 1>}
 base_hair_02_l {parent=dyn_hair_01_l, loc=<4.6898 0 0>, rot=<0 0 0 1>}
 dyn_hair_02_l {parent=base_hair_02_l, loc=<0 0 0>, rot=<0 0 0 1>}
 dyn_hair_03_l {parent=dyn_hair_02_l, loc=<4.6898 0 0>, rot=<0 0 0 1>}
 dyn_hair_04_l {parent=dyn_hair_03_l, loc=<4.6898 0 0>, rot=<0 0 0 1>}
 hairSmall_master_l {parent=dyn_hair_04_l, loc=<-4.3473 0.8236 0.5622>, rot=<0.6247 0.3432 -0.1991 -0.6725>}
 hairSmall_2_01_l {parent=hairSmall_master_l, loc=<1.2906 -0.1159 1.7891>, rot=<0.1622 0 0 -0.9868>}
 hairSmall_1_01_l {parent=hairSmall_master_l, loc=<2.0866 0.4135 -1.1443>, rot=<0.108 -0.1868 0.115 -0.9696>}
 base_hair_front_01 {parent=head, loc=<11.1896 9.7705 3.7857>, rot=<0.2835 -0.6074 0.735 0.102>}
 dyn_hair_front_01 {parent=base_hair_front_01, loc=<0 0 0>, rot=<0 0 0 1>}
 base_hair_front_02 {parent=head, loc=<11.2262 7.2482 4.2424>, rot=<0.1656 -0.7117 0.6769 0.0886>}
 dyn_hair_front_02 {parent=base_hair_front_02, loc=<0 0 0>, rot=<0 0 0 1>}
 clavicle_l {parent=spine_05, loc=<4.3575 -0.1466 -2.3223>, rot=<0.0465 -0.7289 -0.009 -0.6829>}
 dyn_clavicle_mover_l {parent=clavicle_l, loc=<0 0 0>, rot=<0 0 0 1>}
 clavicle_r {parent=spine_05, loc=<4.3575 -0.1466 2.3223>, rot=<0.7289 0.0465 -0.6829 0.009>}
 dyn_clavicle_mover_r {parent=clavicle_r, loc=<0 0 0>, rot=<0 0 0 1>}
 spine_05_ctrl {controlType=EULER_TRANSFORM, type=Control, loc=<0 0 0>, rot=<0 0 0 1>}
 neck_01_ctrl {controlType=EULER_TRANSFORM, parent=spine_05_ctrl, type=Control, loc=<0 0 0>, rot=<0 0 0 1>}
 neck_02_ctrl {controlType=EULER_TRANSFORM, parent=neck_01_ctrl, type=Control, loc=<0 0 0>, rot=<0 0 0 1>}
 head_ctrl {controlType=EULER_TRANSFORM, parent=neck_02_ctrl, type=Control, loc=<0 0 0>, rot=<0 0 0 1>}
 clavicle_l_ctrl {controlType=EULER_TRANSFORM, parent=spine_05_ctrl, type=Control, loc=<0 0 0>, rot=<0 0 0 1>}
 clavicle_r_ctrl {controlType=EULER_TRANSFORM, parent=spine_05_ctrl, type=Control, loc=<0 0 0>, rot=<0 0 0 1>}
]
variables [
 Crouch Walking Alpha {type=double, default=0}
 Is Targeting Crouch {type=bool, default=False}
 Is Targeting Crouch Walk {type=double, default=0}
]
