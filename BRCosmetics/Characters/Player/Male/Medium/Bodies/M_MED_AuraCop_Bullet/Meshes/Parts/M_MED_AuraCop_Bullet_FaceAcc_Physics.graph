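# Physics asset for the face-accessory part of the male medium AuraCop Bullet character, set up for secondary motion on its hair.
# Three capsule chains (left, middle, right) run from base hair bones on the head through three dyn_hair bones, with tight per-side swing limits and almost no twist.
# Capsules on head, neck, upper spine and clavicle movers form the body shapes around the hair, linked by free-angle constraints.
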PhysicsAsset /BRCosmetics/Characters/Player/Male/Medium/Bodies/M_MED_AuraCop_Bullet/Meshes/Parts/M_MED_AuraCop_Bullet_FaceAcc_Physics
bodies {
 base_hair_02_l [{shapeType=Capsule, length=0.505, radius=0.505, shapeName=base_hair_02_l_capsule, rotation=<rot 0 -90 0>}]
 base_hair_02_m [{shapeType=Capsule, length=0.505, radius=0.505, shapeName=base_hair_02_m_capsule, rotation=<rot 0 -90 0>}]
 base_hair_02_r [{shapeType=Capsule, length=0.505, radius=0.505, shapeName=base_hair_02_r_capsule, rotation=<rot 0 -90 0>}]
 dyn_clavicle_mover_l [{
   shapeType=Capsule
   length=6.8534
   radius=7.5754
   shapeName=clavicle_r_capsule
   center=<8.4577 -0.2243 1.2572>
   rotation=<rot 77.7472 -0.0014 -0.0014>
  }]
 dyn_clavicle_mover_r [{
   shapeType=Capsule
   length=6.8534
   radius=7.5754
   shapeName=clavicle_r_capsule
   center=<-8.4577 0.2243 -1.2572>
   rotation=<rot 77.7472 -0.0014 -0.0014>
  }]
 dyn_hair_02_l [{shapeType=Capsule, length=6.6977, radius=1.7687, shapeName=dyn_hair_02_l_capsule, center=<2 -0.1165 0.1664>, rotation=<rot 0 0 55>}]
 dyn_hair_02_m [{shapeType=Capsule, length=6.6977, radius=1.7687, shapeName=dyn_hair_02_m_capsule, center=<2 0 0>}]
 dyn_hair_02_r [{shapeType=Capsule, length=6.6977, radius=1.7687, shapeName=dyn_hair_02_r_capsule, center=<2 -0.1165 -0.1664>, rotation=<rot 0 0 -55>}]
 dyn_hair_03_l [{shapeType=Capsule, length=6.6184, radius=1.848, shapeName=dyn_hair_03_l_capsule, center=<1.4483 -0.3302 1.4198>, rotation=<rot 0 0 55>}]
 dyn_hair_03_m [{shapeType=Capsule, length=6.6184, radius=1.5, shapeName=dyn_hair_03_m_capsule, center=<2 0 0>}]
 dyn_hair_03_r [{shapeType=Capsule, length=6.6184, radius=1.8, shapeName=dyn_hair_03_r_capsule, center=<1.0805 -0.8453 -1.2072>, rotation=<rot 0 0 -55>}]
 dyn_hair_04_l [{
   shapeType=Capsule
   length=6.8164
   radius=1.65
   shapeName=dyn_hair_04_l_capsule
   center=<-0.1195 -1.9647 2.428>
   rotation=<rot 0 0.0002 29.1527>
  }]
 dyn_hair_04_m [{shapeType=Capsule, length=6.8164, radius=1.65, shapeName=dyn_hair_04_m_capsule, center=<1.6994 -1.0927 0>}]
 dyn_hair_04_r [{shapeType=Capsule, length=6.8164, radius=1.65, shapeName=dyn_hair_04_r_capsule, center=<0.1449 -0.9909 -2.1844>, rotation=<rot 0 0 -55>}]
 head [{shapeType=Capsule, length=1.0404, radius=6.6587, shapeName=head_capsule, center=<5.0606 2.6748 0>, rotation=<rot 90 0 0>}]
 neck_01 [{shapeType=Capsule, length=0.9591, radius=7.5, shapeName=neck_01_capsule, center=<2 1.6508 0>, rotation=<rot 90 0 0>}]
 neck_02 [{shapeType=Capsule, length=3.9458, radius=5.5656, shapeName=neck_02_capsule, center=<2.3715 1.4221 0>, rotation=<rot 90 0 0>}]
 spine_05 [
  {shapeType=Capsule, length=10, radius=10, shapeName=spine_05_capsule, center=<3.1307 -0.4609 0>, rotation=<rot 0 -90 0>}
  {shapeType=Capsule, length=6.5, radius=10, shapeName=spine_05_capsule, center=<5.2595 0.9372 0>, rotation=<rot 0 -90 0>}
 ]
}
constraints [
 {bone1Name=dyn_hair_02_m, bone2Name=base_hair_02_m, swing1LimitDegrees=15, swing2LimitDegrees=10, twistLimitDegrees=1}
 {bone1Name=dyn_hair_03_m, bone2Name=dyn_hair_02_m, swing1LimitDegrees=20, swing2LimitDegrees=10, twistLimitDegrees=1}
 {bone1Name=dyn_hair_04_m, bone2Name=dyn_hair_03_m, swing1LimitDegrees=12, swing2LimitDegrees=10, twistLimitDegrees=1}
 {bone1Name=dyn_hair_04_l, bone2Name=dyn_hair_03_l, swing1LimitDegrees=12, swing2LimitDegrees=10, twistLimitDegrees=1}
 {bone1Name=dyn_hair_03_l, bone2Name=dyn_hair_02_l, swing1LimitDegrees=10, swing2LimitDegrees=10, twistLimitDegrees=1}
 {bone1Name=dyn_hair_02_l, bone2Name=base_hair_02_l, swing1LimitDegrees=8, swing2LimitDegrees=10, twistLimitDegrees=1}
 {bone1Name=dyn_hair_02_r, bone2Name=base_hair_02_r, swing1LimitDegrees=2.5, swing2LimitDegrees=10, twistLimitDegrees=1}
 {bone1Name=dyn_hair_04_r, bone2Name=dyn_hair_03_r, swing1LimitDegrees=7, swing2LimitDegrees=15, twistLimitDegrees=1}
 {bone1Name=dyn_hair_03_r, bone2Name=dyn_hair_02_r, swing1LimitDegrees=5, swing2LimitDegrees=10, twistLimitDegrees=1}
 {
  bone1Name=head
  bone2Name=neck_02
  swing1LimitDegrees=45
  swing1Motion=Free
  swing2LimitDegrees=45
  swing2Motion=Free
  twistLimitDegrees=45
  twistMotion=Free
 }
 {
  bone1Name=neck_02
  bone2Name=neck_01
  swing1LimitDegrees=45
  swing1Motion=Free
  swing2LimitDegrees=45
  swing2Motion=Free
  twistLimitDegrees=45
  twistMotion=Free
 }
 {
  bone1Name=base_hair_02_l
  bone2Name=head
  swing1LimitDegrees=45
  swing1Motion=Free
  swing2LimitDegrees=45
  swing2Motion=Free
  twistLimitDegrees=45
  twistMotion=Free
 }
 {
  bone1Name=base_hair_02_m
  bone2Name=head
  swing1LimitDegrees=45
  swing1Motion=Free
  swing2LimitDegrees=45
  swing2Motion=Free
  twistLimitDegrees=45
  twistMotion=Free
 }
 {
  bone1Name=base_hair_02_r
  bone2Name=head
  swing1LimitDegrees=45
  swing1Motion=Free
  swing2LimitDegrees=45
  swing2Motion=Free
  twistLimitDegrees=45
  twistMotion=Free
 }
 {
  bone1Name=neck_01
  bone2Name=spine_05
  swing1LimitDegrees=45
  swing1Motion=Free
  swing2LimitDegrees=45
  swing2Motion=Free
  twistLimitDegrees=45
  twistMotion=Free
 }
 {
  bone1Name=dyn_clavicle_mover_l
  bone2Name=spine_05
  swing1LimitDegrees=45
  swing1Motion=Free
  swing2LimitDegrees=45
  swing2Motion=Free
  twistLimitDegrees=45
  twistMotion=Free
 }
 {
  bone1Name=dyn_clavicle_mover_r
  bone2Name=spine_05
  swing1LimitDegrees=45
  swing1Motion=Free
  swing2LimitDegrees=45
  swing2Motion=Free
  twistLimitDegrees=45
  twistMotion=Free
 }
]
