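# Post-process style secondary-motion AnimBP for the AuraCop Bullet face accessory mesh, mainly its hair.
# It copies the parent mesh pose into a Control Rig driven by crouch-moving and crouch-targeting states.
# RigidBody simulation in spine_05 space then feeds three chained head-relative AnimDynamics front-hair bones, using a computed gravity override on all but the last.
# The cached result gets an additive left-hair bone offset only in the front end.
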AnimBlueprint /BRCosmetics/Characters/Player/Male/Medium/Bodies/M_MED_AuraCop_Bullet/Meshes/Parts/M_MED_AuraCop_Bullet_FaceAcc_AnimBP
target_skeleton=&/M_MED_AuraCop_Bullet_FaceAcc_Skeleton
graphs [
 AnimGraph {nodes [
   n1 = "Transform (Modify) Bone - Bone: hairSmall_master_l"(
    ComponentPose=Local To Component(LocalPose=Use cached pose 'BaseDynamics'())
    Alpha=1
    Translation="-2,0,0"
    Rotation="0,-5,0"
    Scale="1,1,1"
   ) {
    rotationMode=BMM_Additive
    rotationSpace=BCS_BoneSpace
    translationMode=BMM_Additive
    translationSpace=BCS_BoneSpace
    alphaBoolBlend {blendOption=Linear}
    boneToModify {boneName=hairSmall_master_l}
    rotation {yaw=-5}
    translation {x=-2}
   }
   Output Pose(Result=Blend Poses by bool(BlendPose_0=Component To Local(ComponentPose=@n1), BlendPose_1=Use cached pose 'BaseDynamics'(), bActiveValue=Get bIsInFrontEnd()) {blendType=Linear}) {layerGroup=None}
   n2 = Control Rig!(
    Source=Copy Pose From Mesh() {bUseAttachedParent=true}
    Alpha
    Crouch Walking Alpha:double=Get Is Crouch Moving() -> (Is Crouch Moving:double)
    Is Targeting Crouch:bool=Get Is Targeting Crouch() -> (Is Targeting Crouch:bool)
   ) -> (Pose) {
    alphaBoolBlend {blendOption=Linear}
    controlRigAssetReference {blueprintRigClass=&/M_MED_AuraCop_Bullet_FaceAcc_CtrlRig.M_MED_AuraCop_Bullet_FaceAcc_CtrlRig_C}
   }
   n3 = Get GravityOverride()
   n4 = RigidBody!(ComponentPose=Local To Component(LocalPose=@n2), Alpha, OverrideWorldGravity:Vector=@n3, ExternalForce) -> (Pose) {
    bOverrideWorldGravity=true
    overridePhysicsAsset=&/M_MED_AuraCop_Bullet_FaceAcc_Physics
    simulationSpace=BaseBoneSpace
    alphaBoolBlend {blendOption=Linear}
    baseBoneRef {boneName=spine_05}
    componentAppliedLinearAccClamp=<2800 2800 2800>
    componentLinearAccScale=<0.28 0.28 0.28>
    componentLinearVelScale=<0.28 0.28 0.28>
    simSpaceSettings {
     maxAngularAcceleration=1000
     maxAngularVelocity=3
     maxLinearAcceleration=700
     maxLinearVelocity=500
     velocityScaleZ=0
     worldAlpha=0.2
     externalLinearDragV=<0.2 0.2>
    }
   }
   n5 = Anim Dynamics - Bone: dyn_hair_front_01!(ComponentPose=@n4, bAlphaBoolEnabled:bool=True, GravityOverride:Vector=@n3, ExternalForce="0,0,0") -> (Pose) {
    alphaInputType=Bool
    angularDampingOverride=0.9999
    angularSpringConstant=2400
    bAngularSpring=true
    bOverrideAngularDamping=true
    bUseGravityOverride=true
    gravityScale=0.35
    simulationSpace=RootRelative
    alphaBoolBlend {blendInTime=2, blendOption=Cubic}
    boundBone {boneName=dyn_hair_front_01}
    componentAppliedLinearAccClamp=<3000 3000 3000>
    componentLinearAccScale=<0.3 0.4 0.3>
    componentLinearVelScale=<0.3 0.4 0.3>
    gravityOverride {z=-980}
    physicsBodyDefinitions [{
      boundBone {boneName=dyn_hair_front_01}
      boxExtents=<25 25 25>
      constraintSetup {angularLimitsMax=<0 45 10>, angularLimitsMin=<0 -5 -50>, angularTarget=<1 0 0>}
      localJointOffset=<5 0 0>
     }]
    relativeSpaceBone {boneName=head}
   }
   n6 = Anim Dynamics - Bone: dyn_hair_front_02!(ComponentPose=@n5, bAlphaBoolEnabled:bool=True, GravityOverride:Vector=@n3, ExternalForce="0,0,0") -> (Pose) {
    alphaInputType=Bool
    angularDampingOverride=0.9999
    angularSpringConstant=2400
    bAngularSpring=true
    bOverrideAngularDamping=true
    bUseGravityOverride=true
    gravityScale=0.35
    simulationSpace=RootRelative
    alphaBoolBlend {blendInTime=2, blendOption=Cubic}
    boundBone {boneName=dyn_hair_front_02}
    componentAppliedLinearAccClamp=<3000 3000 3000>
    componentLinearAccScale=<0.3 0.4 0.3>
    componentLinearVelScale=<0.3 0.4 0.3>
    gravityOverride {z=-980}
    physicsBodyDefinitions [{
      boundBone {boneName=dyn_hair_front_02}
      boxExtents=<25 25 25>
      constraintSetup {angularLimitsMax=<0 17 0>, angularLimitsMin=<0 -8 -40>, angularTarget=<1 0 0>}
      localJointOffset=<3 0 0>
     }]
    relativeSpaceBone {boneName=head}
   }
   n7 = Anim Dynamics - Bone: dyn_hair_front_03!(ComponentPose=@n6, bAlphaBoolEnabled:bool=True, GravityOverride:Vector="0,0,0", ExternalForce="0,0,0") -> (Pose) {
    alphaInputType=Bool
    angularDampingOverride=0.9999
    angularSpringConstant=2400
    bAngularSpring=true
    bOverrideAngularDamping=true
    bUseGravityOverride=true
    gravityScale=0.35
    simulationSpace=RootRelative
    alphaBoolBlend {blendInTime=2, blendOption=Cubic}
    boundBone {boneName=dyn_hair_front_03}
    componentAppliedLinearAccClamp=<3000 3000 3000>
    componentLinearAccScale=<0.3 0.4 0.3>
    componentLinearVelScale=<0.3 0.4 0.3>
    gravityOverride {z=-980}
    physicsBodyDefinitions [{
      boundBone {boneName=dyn_hair_front_03}
      boxExtents=<25 25 25>
      constraintSetup {angularLimitsMax=<0 30 2>, angularLimitsMin=<0 -60 -40>, angularTarget=<1 0 0>}
      localJointOffset=<3 0 0>
     }]
    relativeSpaceBone {boneName=head}
   }
   Save cached pose 'BaseDynamics'(Pose=Component To Local(ComponentPose=@n7)) {cacheName=BaseDynamics}
  ]}
 EventGraph {nodes [
   Event BlueprintUpdateAnimation()
   > n1 = Sequence!(execute) -> (then_0, then_1, then_2:Exec)
   > n2 = Gravity Override State Function DEPRECATED(CharacterPartAnimInstance=Self-Reference(), GameTimeInSeconds=GetGameTimeInSeconds(), GravityOverrideParams=Get Gravity Override Params())
   > Set GravityOverride(GravityOverride=@n2)
   n3 = Get bIsCrouching()
   n4 = Get bIsMoving2D()
   Set Is Crouch Moving(execute=@n1.then_1, Is Crouch Moving:double="To Float (Boolean)"(InBool=AND Boolean(A=@n3, B=@n4))) -> (Output_Get:double)
   Set Is Targeting Crouch(execute=@n1.then_2, Is Targeting Crouch:bool=AND Boolean(A=Get bIsCrouching(), B=Get bIsTargeting())) -> (Output_Get:bool)
   Get isTargeting() -> (isTargeting:double)
   n5 = Get bIsTargeting()
   Set isTargeting(isTargeting:double=FInterpTo(Current=Get isTargeting(), Target="To Float (Boolean)"(InBool=@n5), DeltaTime=0, InterpSpeed=8)) -> (Output_Get:double)
  ]}
]
variables [
 Gravity Override Params {type=GravityOverrideParamsStruct}
 GravityOverride {type=Vector}
 headUp {type=number}
 Is Crouch Moving {type=number}
 Is Targeting Crouch {type=boolean}
 isTargeting {type=number}
]
